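# Niagara system for the Neonshapes vehicle boost cosmetic, pairing neon ribbon trails with small sprite particles spawned at skeletal mesh sockets and thrown backward.
# Both emitters take their color from painted-variant overrides and inherit source movement, which fades over particle life.
# VC_Mask modules vary spawn rate and velocity across Coro, FrontEnd and FN contexts.
# User.bIsFrontendPreview further adjusts default color, size and speed, plus ribbon tiling and contrast.
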NiagaraSystem /VehicleCosmetics/Boosts/Boost_Neonshapes/FX/NS_Neonshapes
dependencies {
 usedDataInterfaces [&NiagaraDataInterfaceCurve, &NiagaraDataInterfaceSkeletalMesh, &NiagaraDataInterfaceArrayFloat3]
 usedDynamicInputs [&SelectVectorFromArray]
 usedModules [
  &self:SystemSpawnScript.SystemScriptSource.SystemScriptGraph.NiagaraNodeAssignment_3.SetVariables_#1
  &self:SystemSpawnScript.SystemScriptSource.SystemScriptGraph.NiagaraNodeAssignment_4.SetVariables_#2
  &self:SystemSpawnScript.SystemScriptSource.SystemScriptGraph.NiagaraNodeAssignment_2.SetVariables_#3
  &SystemState
 ]
}
emitters {
 NE_VC_Boost_Ribbon_Parent {
  emitterUpdateScript [
   &EmitterState()
   &VC_FilteredSocketCount()
   &VC_RateQualityHandler(Default Rate:NiagaraFloat=&Multiply_Float(
     A=100
     B=/VehicleCosmetics/FX/Niagara/Modules/VC_Mask_Floats(Coro Value:NiagaraFloat=1, FrontEnd Value:NiagaraFloat=1, FN Value:NiagaraFloat=0.75)
    ))
   &VC_SpawnRibbon_v1(Number of Ribbons=$Emitter.PortCount, Spawn Time=-1, SpawnRate=$Emitter.Spawn Rate)
   /SetVariables(Emitter.TilingX:NiagaraFloat=&MakeCustomFloatFromBool(Bool=$User.bIsFrontendPreview, True Float=1, False Float=1.5)) {isSetParameters=true}
   /SetVariables(Emitter.ContrastTEX1:NiagaraFloat=&MakeCustomFloatFromBool(Bool=$User.bIsFrontendPreview, True Float=2, False Float=5)) {isSetParameters=true}
  ]
  particleSpawnScript [
   &VC_PaintedVariantsOverride(
    Black=<rgba 0 0 0 10>
    BurntSienna=<rgba 0.41 0.1777 0 4>
    Cobalt=<rgba 0 0.185 10 5>
    Crimson=<rgba 5 0 0.1355 5>
    DefaultColor=&MakeCustomLinearColorFromBool(Bool=$User.bIsFrontendPreview, True Color=<rgba 28.25 9 30 1>, False Color=<rgba 27.875 4.5 30 1>)
    ForestGreen=<rgba 0 10 0.5053 5>
    Grey=<rgba 0.35 0.35 0.35 5>
    Lime=<rgba 1.75 3 0 5>
    Orange=<rgba 3 1.5 0 5>
    Purple=<rgba 4.1667 0 10 5>
    Saffron=<rgba 3 2.3 0 5>
    SkyBlue=<rgba 0 1.9962 3 5>
    TitaniumWhite=<rgba 5 5 5 5>
   )
   &VC_InitializeRibbon_v1(
    Color=$Particles.Color
    Custom Offset=true
    Enable Skeletal Position=true
    Facing Axis:ENiagaraOrientationAxis=::ZAxis
    Lifetime=0.45
    Local Vector Offset=<20 0 0>
    Offset Amount=1
    Ribbon Width:NiagaraFloat=&MakeCustomFloatFromBool(Bool=$User.bIsFrontendPreview, True Float=50, False Float=60)
   )
   &AddVelocity(Velocity=&VC_Mask_Vectors(
     Coro Value:Vector3f=<-1000 0 0>
     FrontEnd Value:Vector3f=&MakeCustomVectorFromBool(Bool=$User.bIsFrontendPreview, True Vector=<-2500 0 1>, False Vector=<-1500 0 0>)
     FN Value:Vector3f=<-1200 0 0>
    )) {module=AddVelocity001}
   /SetVariables(Particles.RibbonTwist=0.5) {isSetParameters=true}
   /VehicleCosmetics/FX/VC_FXParents/Modules/DLR_ProductPreviews(GameModeID:NiagaraInt32=$System.GameModeID, bPreviewVehicleSpeed:NiagaraBool=$System.bPreviewVehicleSpeed)
  ]
  particleUpdateScript [
   &ParticleState()
   &InheritSourceMovement(
    Applied Position Scale=&VectorFromFloat(Value=&FloatFromCurve(FloatCurve=DataInterface() {propertyValues {Curve {
         defaultValue=1
         keys [
          {arriveTangent=-1.7885, interpMode=RCIM_Cubic, leaveTangent=-1.7885, tangentMode=RCTM_User, time=0.05, value=1}
          {interpMode=RCIM_Cubic, tangentMode=RCTM_User, time=1, value=0}
         ]
        }}}))
    ApplyForce=false
    ApplyPosition=true
    ApplyPositionFalloff=false
   )
   &SolveForcesAndVelocity(Clamp Velocity=false, Limit Acceleration=false)
   &VC_UpdateRibbon_v1(Enable Skeletal Position=true)
   &ScaleColor(
    Scale RGB=&Multiply_VectorByFloat(
     Vector=<1 1 1>
     Float=&FloatFromCurve(FloatCurve=DataInterface() {propertyValues {Curve {keys [
          {
           arriveTangent=185.9036
           arriveTangentWeight=30.9844
           interpMode=RCIM_Cubic
           leaveTangent=4.892
           leaveTangentWeight=30.9844
           tangentMode=RCTM_Break
           time=0
           value=1
          }
          {
           arriveTangent=-0.1664
           arriveTangentWeight=5.0853
           leaveTangent=-30.4954
           leaveTangentWeight=0.2592
           tangentMode=RCTM_User
           time=0.5
           value=2
          }
          {arriveTangent=-171.254, arriveTangentWeight=0.7935, leaveTangent=5.1948, leaveTangentWeight=0.7935, time=0.95, value=1}
         ]}}})
    )
    ScaleA=true
    ScaleRGB=true
   )
  ]
  rendererClasses [&NiagaraRibbonRendererProperties]
  renderers [{rendererClass=&NiagaraRibbonRendererProperties}]
 }
 NE_VC_Boost_Sprite_Parent {
  emitterUpdateScript [
   &EmitterState()
   &VC_FilteredSocketCount()
   &SpawnRate(
    SpawnRate=&Multiply_Float(
     A=100
     B=&VC_Mask_Floats(
      Coro Value:NiagaraFloat=0.5
      FrontEnd Value:NiagaraFloat=&MakeCustomFloatFromBool(Bool=$User.bIsFrontendPreview, True Float=0.6, False Float=1)
      FN Value:NiagaraFloat=0.5
     )
    )
    Use Spawn Probability=false
   )
  ]
  particleSpawnScript [
   &VC_PaintedVariantsOverride(
    Black=<rgba 0 0 0 1>
    BurntSienna=<rgba 0.4 0.1 0 1>
    Cobalt=<rgba 5 6.7575 100 1>
    Crimson=<rgba 50 0 1.3552 1>
    DefaultColor=&MakeCustomLinearColorFromBool(Bool=$User.bIsFrontendPreview, True Color=<rgba 46.25 5 50 1>, False Color=<rgba 8 0.8 8 1>)
    ForestGreen=<rgba 0 50 2.5263 1>
    Grey=<rgba 0.35 0.35 0.35 1>
    Lime=<rgba 17.5 30 0 1>
    Orange=<rgba 30 15 0 1>
    Purple=<rgba 41.6666 0 100 1>
    Saffron=<rgba 30 23 0 1>
    SkyBlue=<rgba 0 19.9616 30 1>
    TitaniumWhite=<rgba 50 50 50 1>
   )
   &InitializeParticle(
    Color=$Particles.Color
    Lifetime=&RandomRangeFloat(Minimum=0.2, Maximum=0.3)
    Sprite Rotation Mode=::NewEnumerator2
    Sprite Size Mode=::NewEnumerator1
    Uniform Sprite Size Max=50
    Uniform Sprite Size Min=30
    UsePositionOffset=false
    Write Material Random=true
   )
   &AddVelocity(Module.Use Falloff Distance=false, Override Seed=false, Velocity Mode=::NewEnumerator1, Velocity Speed=150) {module=AddVelocity001}
   &PartitionParticles(Number Of Partitions=1)
   &SkeletalMeshLocation(
    Bone / Socket Index=$Particles.PartitionParticles.Partition
    Bone / Socket Sampling Mode=::NewEnumerator5
    MaskSpawnGroup=false
    Mesh Sampling Type=::NewEnumerator6
    Sampled Position Offset=<0 -30 0>
    Skeletal Mesh=$System.SkeletalMesh
   )
   &ShapeLocation(MaskSpawnGroup=false, Non Uniform Scale=<1 1 0.4>, Override Seed=false, Sphere Radius=50, UseOwnerScale=true)
   &AddVelocity(Velocity=&VC_Mask_Vectors(
     Coro Value:Vector3f=<-1000 0 0>
     FrontEnd Value:Vector3f=&MakeCustomVectorFromBool(Bool=$User.bIsFrontendPreview, True Vector=<-2200 0 1>, False Vector=<-1200 0 0>)
     FN Value:Vector3f=<-400 0 0>
    ))
   &DLR_ProductPreviews(GameModeID:NiagaraInt32=$System.GameModeID, bPreviewVehicleSpeed:NiagaraBool=$System.bPreviewVehicleSpeed)
  ]
  particleUpdateScript [
   &ParticleState()
   &ScaleSpriteSize(
    Uniform Curve Scale=&MakeCustomFloatFromBool(Bool=$User.bIsFrontendPreview, True Float=0.6, False Float=1)
    Uniform Curve Sprite Scale=DataInterface() {propertyValues {Curve {
       defaultValue=0
       keys [
        {arriveTangent=-0.8665, interpMode=RCIM_Cubic, time=0.65, value=1}
        {arriveTangent=-6.6369, leaveTangent=-0.8314, tangentMode=RCTM_User, time=1, value=0}
       ]
      }}}
   )
   &InheritSourceMovement(
    Applied Position Scale=&VectorFromFloat(Value=&FloatFromCurve(FloatCurve=DataInterface() {propertyValues {Curve {
         defaultValue=1
         keys [
          {arriveTangent=-1.7885, interpMode=RCIM_Cubic, leaveTangent=-1.7885, tangentMode=RCTM_User, time=0.05, value=1}
          {interpMode=RCIM_Cubic, tangentMode=RCTM_User, time=1, value=0}
         ]
        }}}))
    ApplyForce=false
    ApplyPosition=true
    ApplyPositionFalloff=false
   )
   &SolveForcesAndVelocity(Clamp Velocity=false, Limit Acceleration=false)
  ]
  rendererClasses [&NiagaraSpriteRendererProperties]
  renderers [{rendererClass=&NiagaraSpriteRendererProperties}]
 }
}
summary {
 systemName=NS_Neonshapes
 emitters [
  {emitterName=NE_VC_Boost_Ribbon_Parent, rendererClasses [&NiagaraRibbonRendererProperties]}
  {emitterName=NE_VC_Boost_Sprite_Parent, rendererClasses [&NiagaraSpriteRendererProperties]}
 ]
 userVariables [
  User.bIsFrontendPreview {type=NiagaraBool}
  User.bIsPainted {type=NiagaraBool}
  User.blsIsCoroColor {type=NiagaraBool}
  User.bPreviewVehicleSpeed {type=NiagaraBool}
  User.ColorVariantID {type=NiagaraInt32}
  User.GameModeID {type=NiagaraInt32}
  User.PaintedColor {type=LinearColor, description="", defaultValue {struct=LinearColor, value=<rgba 1 1 1 1>}}
  User.PaintedScalar {type=NiagaraFloat, defaultValue=0, description=""}
  User.PortCount {type=NiagaraInt32, defaultValue=2}
 ]
}
